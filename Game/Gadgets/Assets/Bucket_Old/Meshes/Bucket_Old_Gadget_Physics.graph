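# Physics asset for the old bucket gadget's skeletal mesh, defining its collision and jointed physics as two bodies.
# The bucket and lid bones each carry a single offset, rotated box as a simple collision approximation.
# One constraint joins the lid to the bucket, with separate swing and twist angle limits, widest on the first swing axis and tightest on twist.
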PhysicsAsset /Game/Gadgets/Assets/Bucket_Old/Meshes/Bucket_Old_Gadget_Physics
bodies {
 bucket [{shapeType=Box, extentX=7.7801, extentY=6.1095, extentZ=4.4545, center=<0.2664 -0.2197 -0.4525>, rotation=<rot -1.5481 -90 3.8535>}]
 lid [{shapeType=Box, extentX=5.3264, extentY=10.3692, extentZ=12.9212, center=<-2.3119 0.812 0.6319>, rotation=<rot 0.2202 -94.5154 8.8822>}]
}
constraints [{bone1Name=lid, bone2Name=bucket, swing1LimitDegrees=45, swing2LimitDegrees=20, twistLimitDegrees=12}]
