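# Animation Blueprint for the old bucket gadget that adds physics-driven secondary motion to its skeleton.
# The input pose runs through a RigidBody node using the gadget's physics asset, with overridden downward gravity and Root as the base bone.
# An additive 90-degree yaw on Root then feeds the output pose.
# An AnimDynamics spring node for the lid bone is left unconnected, and the event graph has no logic.
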AnimBlueprint /Game/Gadgets/Assets/Bucket_Old/Meshes/Bucket_Old_Gadget_Skeleton_AnimBP
target_skeleton=&/Bucket_Old_Gadget_Skeleton
graphs [
 AnimGraph {nodes [
   Anim Dynamics - Bone: lid(Alpha=1, ExternalForce="0,0,0") {
    bAngularSpring=true
    bGravityOverrideInSimSpace=true
    alphaBoolBlend {blendOption=Linear}
    boundBone {boneName=lid}
    componentLinearAccScale=<0.5 0.5 0.5>
    componentLinearVelScale=<0.2 0.2 0.2>
    physicsBodyDefinitions [{
      sphereCollisionRadius=0
      boundBone {boneName=lid}
      boxExtents=<4 4 4>
      constraintSetup {angularLimitsMax=<0 0 0>, angularLimitsMin=<0 0 0>, angularTarget=<1 0 0>}
      localJointOffset=<-0.1 -2 0>
     }]
   }
   AnimGraphNode_SubInput_0 = Input Pose()
   n1 = RigidBody(ComponentPose=Local To Component(LocalPose=@AnimGraphNode_SubInput_0)) {
    bClampLinearTranslationLimitToRefPose=true
    bOverrideWorldGravity=true
    overridePhysicsAsset=&/Bucket_Old_Gadget_Physics
    alphaBoolBlend {blendOption=Linear}
    baseBoneRef {boneName=Root}
    componentLinearAccScale=<1 1 1>
    componentLinearVelScale=<1 1 1>
    overrideWorldGravity {z=-980}
   }
   Output Pose(Result=Component To Local(ComponentPose="Transform (Modify) Bone - Bone: Root"(ComponentPose=@n1, Alpha=1, Translation="0,0,0", Rotation="0,90,0", Scale="1,1,1") {rotationMode=BMM_Additive, alphaBoolBlend {blendOption=Linear}, boneToModify {boneName=Root}, rotation {yaw=90}})) {layerGroup=None}
  ]}
 EventGraph {nodes [
   TryGetPawnOwner()
   Event BlueprintUpdateAnimation()
  ]}
]
variables []
